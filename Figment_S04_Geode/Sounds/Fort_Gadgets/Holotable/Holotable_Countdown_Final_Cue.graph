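# Sound cue for the final countdown sound of the Holotable gadget.
# A quality level node selects between two wave players: the first two slots share the standard Countdown_Final_03 wave, and the last two share its "_opt" variant.
# It plays in the Fort_Gadgets sound class at reduced volume, with overridden attenuation setting the falloff distance and stereo spread.
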SoundCue /Figment_S04_Geode/Sounds/Fort_Gadgets/Holotable/Holotable_Countdown_Final_Cue
root=QualityLevel_0
nodes [
 {id=QualityLevel_0, children [WavePlayer_3,WavePlayer_3,WavePlayer_2,WavePlayer_2]}
 {id=WavePlayer_2, properties {soundWaveAssetPtr=&/Holotable_Countdown_Final_03_opt}}
 {id=WavePlayer_3, properties {soundWaveAssetPtr=&/Holotable_Countdown_Final_03}}
]
properties {
 bOverrideAttenuation=true
 maxDistance=2000
 soundClassObject=&Fort_Gadgets
 volumeMultiplier=0.6
 attenuationOverrides {falloffDistance=2000, stereoSpread=256, attenuationShapeExtents {x=0}}
}
