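# Sound cue for the looping launch warning of the Holotable gadget, in the Fort_Gadgets sound class.
# A quality level node selects between the standard looping wave for the first two levels and the "opt" variant, also looping, for the last two.
# Overridden attenuation sets a long falloff, a wide stereo spread and a non-spatialized radius.
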SoundCue /Figment_S04_Geode/Sounds/Fort_Gadgets/Holotable/Holotable_LaunchWarning_Cue
root=QualityLevel_0
nodes [
 {id=QualityLevel_0, children [WavePlayer_0,WavePlayer_0,WavePlayer_1,WavePlayer_1]}
 {id=WavePlayer_0, properties {bLooping=true, soundWaveAssetPtr=&/Holotable_LaunchWarning_04}}
 {
  id=WavePlayer_1
  properties {bLooping=true, soundWaveAssetPtr=&/Holotable_LaunchWarning_04_opt}
 }
]
properties {
 bOverrideAttenuation=true
 maxDistance=6000
 soundClassObject=&Fort_Gadgets
 volumeMultiplier=0.9
 attenuationOverrides {falloffDistance=6000, nonSpatializedRadiusStart=3000, stereoSpread=256, attenuationShapeExtents {x=0}}
}
