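# Sound cue for the Holotable launch sound in the Fort_Gadgets sound class.
# A quality-level node at the root plays the standard launch wave for the first two levels and the "_opt" variant for the last two.
# The cue overrides attenuation with a long falloff and a non-spatialized radius that begins partway out, so the sound loses directionality as the listener gets closer.
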SoundCue /Figment_S04_Geode/Sounds/Fort_Gadgets/Holotable/Holotable_Launch_Cue
root=QualityLevel_0
nodes [
 {id=QualityLevel_0, children [WavePlayer_0,WavePlayer_0,WavePlayer_1,WavePlayer_1]}
 {id=WavePlayer_0, properties {soundWaveAssetPtr=&/Holotable_Launch_05}}
 {id=WavePlayer_1, properties {soundWaveAssetPtr=&/Holotable_Launch_05_opt}}
]
properties {
 bOverrideAttenuation=true
 maxDistance=6000
 soundClassObject=&Fort_Gadgets
 volumeMultiplier=1
 attenuationOverrides {falloffDistance=6000, nonSpatializedRadiusStart=3000, stereoSpread=256, attenuationShapeExtents {x=0}}
}
