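# Sound cue that plays the looping Holotable gadget sound in the Figment_S04_Geode plugin, routed to the Fort_Gadgets sound class.
# A quality-level node feeds two looping wave players: the first two levels use the base-game Holotable_Loop_01 wave, the last two the local Holotable_Loop_01_opt wave.
# The cue overrides attenuation with its own short-range falloff and stereo spread.
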SoundCue /Figment_S04_Geode/Sounds/Fort_Gadgets/Holotable/Holotable_Loop_Cue
root=QualityLevel_0
nodes [
 {id=QualityLevel_0, children [WavePlayer_1,WavePlayer_1,WavePlayer_0,WavePlayer_0]}
 {id=WavePlayer_0, properties {bLooping=true, soundWaveAssetPtr=&/Holotable_Loop_01_opt}}
 {id=WavePlayer_1, properties {bLooping=true, soundWaveAssetPtr=/Game/Sounds/Fort_Gadgets/Holotable/Holotable_Loop_01}}
]
properties {
 bOverrideAttenuation=true
 maxDistance=1000
 soundClassObject=&Fort_Gadgets
 volumeMultiplier=1
 attenuationOverrides {falloffDistance=1000, stereoSpread=256, attenuationShapeExtents {x=0}}
}
